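; ToyFox supply drop pod child Blueprint that falls, lands with impact FX, and spawns boss loot tiered by Solo, Duo, Trio or Squad playlist tags.
; On entering the landed-unopened state, the server marks it for players within a hotfixable distance.
; Falling, landing and doppler sounds are suppressed when the local pawn has blocking tags (ToyFox pilots).
; Interaction time is hotfixable, and the pod dies after spawning loot.
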(event EventBeginPlay
  ;[0 0.15 0.01458 0.5] BeginPlay: Have the pod falling niagara particle be a world rotation 0 0 0.
  (|CallParentFunction)
  (Networking|CanExecuteCosmeticEvents
    (:True
      (Transformation|SetWorldRotation (Variables|Default|GetNS_PodFalling) "0, 0, 0" false false)
      ; Don't play various drop pod sounds for ToyFox pilots
      (if (not (Default|ShouldPlaySounds))
        (Variables|FX|Audio|SetSound_Landed 0)
        (Variables|FX|Audio|SetSound_LandInWater 0)
        (Audio|Components|Audio|Stop (Variables|Default|GetFallingSound_Looping))))))

(event EventSetStateLandedUnopened
  ;[0.1053 0.05668 0.15 0.5] Update chest glow parameters
  (|CallParentFunction)
  (CallFunction|MarkActorforPlayers)
  (bind _chestglowvfx_system (Variables|Default|GetChestGlowVFX_System)
    (:then
      (Effects|Components|ParticleSystem|SetFloatParameter _chestglowvfx_system "Intensity" 0.5)
      (Effects|Components|ParticleSystem|SetFloatParameter _chestglowvfx_system "LightRadiusOverride" 0.3))
    (:else)))

(event EventPlayLandingFX
  (params (Hit :type Hit%20Result%20Structure%20%28by%20ref%29))
  (bind (_bblockinghit _binitialoverlap _time _distance _location) (Collision|BreakHitResult Hit))
  (Utilities|FlowControl|DoOnce false
    (:Completed
      (Niagara|SpawnSystematLocation (Variables|FX|Visual|GetLandImpactVFX) _location)
      (Camera|PlayWorldCameraShake (Variables|FX|GetLandCameraShake) _location 1e+03 3e+03)
      (Niagara|SetNiagaraVariable(Bool) (Variables|Default|GetNS_PodFalling) "FadeOut" true)
      (if (Default|ShouldPlaySounds)
        (Audio|PlaySoundatLocation (Variables|FX|Audio|GetSound_Landed) _location)))))

(event EventDelayedAudioDoppler
  (params (GroundHit :type Hit%20Result%20Structure))
  ; Don't play sound for ToyFox pilots
  (if (Default|ShouldPlaySounds)
    (|CallParentFunction GroundHit)))

(fn BlueprintGetInteractionTime
  (params (InteractingPawn :type Fort%20Pawn%20Object%20Reference) (InteractionBeingAttempted :type EInteractionBeingAttempted%20Enum :default "FirstInteraction"))
  (return (ScalableFloat|GetValueatLevel (Variables|Hotfix|GetHotfix_InteractionTime)) true))

(fn MarkActorForPlayers
  (Networking|SwitchHasAuthority
    (:Authority
      (LocalVariables|SetDistanceSquared (Math|Float|Square (ScalableFloat|GetValueatLevel (Variables|Hotfix|GetHotfix_MarkDistance))))
      (for _array_element (Fortnite|GetAllFortPlayerControllers true true)
        (Utilities|IsValid _array_element
          (:Is%20Valid
            (Utilities|IsValid (Pawn|GetControlledPawn _array_element)
              (:Is%20Valid
                (if (<= (Transformation|GetSquaredDistanceTo (Pawn|GetControlledPawn _array_element) self) (LocalVariables|GetDistanceSquared))
                  (bind _returnvalue (Actor|GetComponentbyClass _array_element "/Script/FortniteGame.AthenaMarkerComponent"))
                  (Utilities|IsValid _returnvalue
                    (:Is%20Valid
                      (MarkerComponent|MarkActoronServer _returnvalue self true true))))))))))))

(fn ShouldPlaySounds
  (bind _returnvalue (Fortnite|GetFirstLocalFortPlayerController true))
  (bind _returnvalue_2 (Game|Player|GetPlayerPawn _returnvalue))
  (Utilities|IsValid _returnvalue
    (:Is%20Valid
      (Utilities|IsValid _returnvalue_2
        (:Is%20Valid
          (return (not (GameplayTags|HasAnyMatchingGameplayTags _returnvalue_2 (Variables|Default|GetSound%20Blocking%20Tags)))))
        (:Is%20Not%20Valid
          (return true))))
    (:Is%20Not%20Valid
      (return true))))

(fn SpawnLoot
  (Utilities|FlowControl|Sequence
    (:then_0
      (bind _returnvalue (Fortnite|Athena|GameState|GetGameStateAthena))
      (Utilities|IsValid _returnvalue
        (:Is%20Valid
          (if (GameplayTags|HasTag (Class|FortGameStateAthena|GetAthenaPlaylistContextTags _returnvalue) "(TagName=\"Athena.Playlist.Solo\")")
            (Variables|DropPodSettings|SetLoot%20Tier%20Name "LTG_Boss_D_Solo")
            (elif (GameplayTags|HasTag (Class|FortGameStateAthena|GetAthenaPlaylistContextTags _returnvalue) "(TagName=\"Athena.Playlist.Duo\")")
              (Variables|DropPodSettings|SetLoot%20Tier%20Name "LTG_Boss_D_Duo")
              (elif (GameplayTags|HasTag (Class|FortGameStateAthena|GetAthenaPlaylistContextTags _returnvalue) "(TagName=\"Athena.Playlist.Trios\")")
                (Variables|DropPodSettings|SetLoot%20Tier%20Name "LTG_Boss_D_Trio")
                (elif (GameplayTags|HasTag (Class|FortGameStateAthena|GetAthenaPlaylistContextTags _returnvalue) "(TagName=\"Athena.Playlist.Squad\")")
                  (Variables|DropPodSettings|SetLoot%20Tier%20Name "LTG_Boss_D_Squad"))))))))
    (:then_1
      (|CallParentFunction)
      (Gameplay|Die))))

(fn ConstructionScript
  (|CallParentFunction)
  (Rendering|Material|SetScalarParameterValueonMaterials (Variables|Debug|GetStaticMeshComponent) "CurrentStage" 0.0))
