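# EQS query that selects ToyFox supply drop spawn locations around the querier.
# A spiral donut generator, sized by scalable-float bindings, creates candidates that a test projects down onto geometry.
# Overlap, water, flat-surface and building tests filter or score them.
# Linear 2D distance scoring favors points farther from the querier, while a filter enforces minimum spacing from TidalCrane ToyFox reposition actors.
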EnvQuery /ToyFoxGameplay/Gameplay/SupplyDrops/EQS_ToyFoxSupplyDropSpawn
options [{
  generator {
   name=EnvQueryGenerator_Donut_0
   class=EnvQueryGenerator_Donut
   properties {
    bUseSpiralPattern=true
    center=&EnvQueryContext_Querier
    innerRadius {
     dataBinding=&self:EnvQueryGenerator_Donut_0.FortAIDataProvider_ScalableFloat_0
     dataField=FloatValue
     defaultValue=3500
    }
    numberOfRings {
     dataBinding=&self:EnvQueryGenerator_Donut_0.FortAIDataProvider_ScalableFloat_2
     dataField=IntValue
     defaultValue=10
    }
    outerRadius {
     dataBinding=&self:EnvQueryGenerator_Donut_0.FortAIDataProvider_ScalableFloat_1
     dataField=FloatValue
     defaultValue=6500
    }
    pointsPerRing {
     dataBinding=&self:EnvQueryGenerator_Donut_0.FortAIDataProvider_ScalableFloat_3
     dataField=IntValue
     defaultValue=16
    }
    projectionData {extentX=0, navigationFilter=null, postProjectionVerticalOffset=0, traceMode=None}
   }
  }
  tests [
   EnvQueryTest_Project_1 {
    class=EnvQueryTest_Project
    properties {
     scoringFactor=1
     testComment=""
     testPurpose=Filter
     projectionData {
      navigationFilter=null
      postProjectionVerticalOffset=0
      projectDown=22000
      projectUp=0
      serializedChannel=ECC_GameTraceChannel11
      traceChannel=TraceTypeQuery10
      traceMode=GeometryByChannel
     }
    }
   }
   EnvQueryTest_Overlap_0 {
    class=EnvQueryTest_Overlap
    properties {overlapData {extentX=150, extentY=150, extentZ=150, overlapChannel=ECC_GameTraceChannel1, shapeOffset=<0 0 300>}}
   }
   EnvQueryTest_Overlap_1 {
    class=EnvQueryTest_Overlap
    properties {testPurpose=FilterAndScore, overlapData {extentX=300, extentY=300, extentZ=300, overlapChannel=ECC_GameTraceChannel15, shapeOffset=<0 0 0>}}
   }
   FortQueryTest_InsideWater_2 {
    class=FortQueryTest_InsideWater
    properties {
     bDefineReferenceValue=false
     boolValue=false
     clampMaxType=None
     clampMinType=None
     filterType=Match
     floatValueMax=0
     floatValueMin=0
     multipleContextFilterOp=AllPass
     multipleContextScoreOp=AverageScore
     normalizationType=Absolute
     referenceValue=0
     scoreClampMax=0
     scoreClampMin=0
     scoringEquation=Constant
     scoringFactor=10
     testComment=""
     testPurpose=Filter
     testRadius=512
    }
   }
   FortQueryTest_OnFlatSurface_1 {class=FortQueryTest_OnFlatSurface, properties {radius=50, toleranceZ=44, traceOffsetUp=100}}
   EnvQueryTest_Volume_0 {
    class=EnvQueryTest_Volume
    properties {boolValue=false, testPurpose=Filter, volumeClass=/Script/FortniteGame.WaterVolume_Athena, volumeContext=null}
   }
   FortQueryTest_HasNearbyBuildings_1 {class=FortQueryTest_HasNearbyBuildings, properties {extentXY=2, testPurpose=FilterAndScore}}
   FortQueryTest_InsideBuilding_0 {class=FortQueryTest_InsideBuilding, properties {testPurpose=FilterAndScore}}
   EnvQueryTest_Distance_3 {
    class=EnvQueryTest_Distance
    properties {
     distanceTo=&EnvQueryContext_Querier
     filterType=Range
     floatValueMax=5000
     floatValueMin=1
     scoringEquation=Linear
     scoringFactor=5
     testComment=""
     testMode=Distance2D
     testPurpose=Score
    }
   }
   EnvQueryTest_Distance_0 {
    class=EnvQueryTest_Distance
    properties {
     distanceTo=&/EnvQueryContext_AllTidalCraneToyFoxRepositionActors.EnvQueryContext_AllTidalCraneToyFoxRepositionActors_C
     filterType=Minimum
     floatValueMin=3000
     scoringEquation=Linear
     scoringFactor=1
     testComment=""
     testMode=Distance2D
     testPurpose=Filter
    }
   }
  ]
 }]
