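; EnvQueryContext that supplies every TidalCrane ToyFox boss reposition actor to EQS queries.
; Its ProvideActorsSet override asks the Fort global actor registry world subsystem for actors registered under the TidalCrane.ToyFox.Boss.Ability.Reposition.Actor gameplay tag and returns that set.
; The querier inputs are unused, so the result is the same regardless of who runs the query.
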(fn ProvideActorsSet
  (params (QuerierObject :type Object%20Reference) (QuerierActor :type Actor%20Object%20Reference))
  (bind _outactors (Class|FortGlobalActorRegistrySubsystem|GetRegisteredActors (WorldSubsystems|GetFortGlobalActorRegistrySubsystem) (GameplayTags|MakeLiteralGameplayTag "(TagName=\"TidalCrane.ToyFox.Boss.Ability.Reposition.Actor\")")))
  (return _outactors))
