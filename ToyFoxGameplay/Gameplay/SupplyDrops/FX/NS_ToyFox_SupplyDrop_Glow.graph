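# Niagara glow effect for ToyFox supply drops, combining a pulsing core, ray flares, sparkles and a smoke trail.
# CorePulse loops one particle driving mesh, sprites and a jittering light; RayFlares adds rotating meshes, Sparkles curl-noise sprites.
# SmokeTrail positions a ten-point ribbon above the owner, lerping with varying lag along its length.
# User parameters tune hue, scale, intensity and pulse speed; time-of-day scalars dim core and flares; Mist, Steam and Rays emitters are empty.
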NiagaraSystem /ToyFoxGameplay/Gameplay/SupplyDrops/FX/NS_ToyFox_SupplyDrop_Glow
dependencies {
 usedDataInterfaces [&NiagaraDataInterfaceCurve, &NiagaraDataInterfaceVectorCurve, &NiagaraDataInterfaceColorCurve]
 usedDynamicInputs [&ClampFloat, &RemapRange, &ColorFromCurve]
 usedModules [
  &self:SystemSpawnScript.SystemScriptSource.SystemScriptGraph.NiagaraNodeAssignment_2.SetVariables_#1
  &SystemState
  &self:SystemSpawnScript.SystemScriptSource.SystemScriptGraph.NiagaraNodeAssignment_1.SetVariables_#2
  &self:SystemSpawnScript.SystemScriptSource.SystemScriptGraph.NiagaraNodeAssignment_0.SetVariables_#3
 ]
}
emitters {
 CorePulse {
  emitterUpdateScript [
   &EmitterState(
    Inactive Response=::NewEnumerator1
    Life Cycle Mode=::NewEnumerator1
    Loop Behavior=::NewEnumerator1
    Loop Duration Mode=::NewEnumerator1
    Reset Age On Awaken=true
    Scalability Mode=::NewEnumerator1
    UseLoopDelay=false
   )
   &SpawnBurst_Instantaneous(Spawn Count=1, Use Spawn Probability=false)
  ]
  particleSpawnScript [
   &InitializeParticle(
    Color=&HueShiftLinearColor(Hue=$User.HueShift, Saturation=1, Value=1, Linear Color=<rgba 8 1.6667 0 0.25>)
    Lifetime=&Divide_Float(A=1.5, B=$User.Pulse_Speed)
    Mesh Scale Mode=::NewEnumerator3
    Mesh Uniform Scale=3
    Position Offset=<15 -40 75>
    Position Offset Coordinate Space=::Local
    Sprite Rotation Mode=::NewEnumerator2
    Sprite Size Mode=::NewEnumerator3
    Uniform Sprite Size=600
    UsePositionOffset=true
    Write Material Random=true
   )
   &InitialMeshOrientation(
    Mesh Orientation Mode=::NewEnumerator2
    Rotation=&RandomRangeVector(Minimum=<-1 0 0>, Maximum=<1 0 0>)
    Use Rotation Vector=true
   )
   /SetVariables(Particles.LightEnabled:NiagaraBool=/Niagara/DynamicInputs/Helpers/MaskBoolBySpawnGroup(Masked Bool:NiagaraBool=false, Passthrough Bool:NiagaraBool=true, Spawn Group Mask:NiagaraInt32=1)) {isSetParameters=true}
  ]
  particleUpdateScript [
   &ParticleState(
    Kill Particles When Lifetime Has Elapsed=false
    Let Infinitely Lived Particles Die When Emitter Deactivates=true
    Loop Particles Lifetime=true
   )
   &ScaleMeshSize(Scale Factor=&VectorFromFloat(Value=&FloatFromCurve(
      FloatCurve=DataInterface() {propertyValues {Curve {
         defaultValue=0
         keys [
          {arriveTangent=4.1785, interpMode=RCIM_Cubic, leaveTangent=4.1785, tangentMode=RCTM_User, time=0, value=0}
          {arriveTangent=1.1179, interpMode=RCIM_Cubic, leaveTangent=1.1179, tangentMode=RCTM_User, time=0.2042, value=0.6228}
          {arriveTangent=0.0224, interpMode=RCIM_Cubic, leaveTangent=0.0224, tangentMode=RCTM_User, time=1, value=1}
         ]
        }}}
      CurveIndex=&NormalizeFloat(Value=&Divide_Float(A=$Particles.NormalizedAge, B=0.666), Min=0, Max=1)
      Scale Curve=$User.FXScale
     )))
   &DynamicMaterialParameters(
    Index 0 Param 3=-0.15
    Module.Param0WriteEnabled=true
    Module.Param1WriteEnabled=true
    Module.Param2WriteEnabled=true
    Module.Param3WriteEnabled=true
   )
   &ScaleColor(
    Scale Alpha=&FloatFromCurve(
     FloatCurve=DataInterface() {propertyValues {Curve {
        defaultValue=1
        keys [
         {interpMode=RCIM_Cubic, tangentMode=RCTM_User, time=0, value=0}
         {interpMode=RCIM_Cubic, tangentMode=RCTM_User, time=0.5, value=1}
         {tangentMode=RCTM_User, time=1, value=0}
        ]
       }}}
     CurveIndex=&NormalizeFloat(Value=&Divide_Float(A=$Particles.NormalizedAge, B=0.666), Min=0, Max=1)
     Scale Curve=&FNTimeOfDayScalar(Night Scale=0.5, Morning Scale=0.9, Evening Scale=0.75)
    )
    ScaleA=true
    ScaleRGB=true
   )
   &ScaleSpriteSize(
    Uniform Curve Scale=$User.FXScale
    Uniform Curve Sprite Scale=DataInterface() {propertyValues {Curve {
       defaultValue=0
       keys [
        {arriveTangent=4.1785, interpMode=RCIM_Cubic, leaveTangent=4.1785, tangentMode=RCTM_User, time=0, value=0}
        {arriveTangent=1.1179, interpMode=RCIM_Cubic, leaveTangent=1.1179, tangentMode=RCTM_User, time=0.2042, value=0.6228}
        {arriveTangent=0.0224, interpMode=RCIM_Cubic, leaveTangent=0.0224, tangentMode=RCTM_User, time=1, value=1}
       ]
      }}}
   )
   /SetVariables(Particles.Color_EnergyWave:LinearColor=&MultiplyLinearColor(
     Linear Color=&HueShiftLinearColor(Hue=$User.HueShift, Saturation=1, Value=1, Linear Color=<rgba 8 1.6667 0 0.25>)
     Scale Factor=&MakeVector4FromVectorAndFloat(
      "Vector (RGB)"=<1 1 1>
      "Float (Alpha)"=&FloatFromCurve(
       FloatCurve=DataInterface() {propertyValues {Curve {
          defaultValue=0
          keys [
           {arriveTangent=1.4628, interpMode=RCIM_Cubic, leaveTangent=1.4628, tangentMode=RCTM_User, time=0, value=0}
           {arriveTangent=-2.5249, interpMode=RCIM_Cubic, leaveTangent=-2.5249, tangentMode=RCTM_User, time=0.3655, value=0.9071}
           {arriveTangent=-0.0405, interpMode=RCIM_Cubic, leaveTangent=-0.0405, tangentMode=RCTM_User, time=1, value=0}
          ]
         }}}
       Scale Curve=&Multiply_Float(A=0.5, B=$User.FX_Intensity)
      )
     )
    )) {isSetParameters=true}
   /SetVariables(
    Particles.Color_Corona:LinearColor=&MultiplyLinearColor(
     Linear Color=&HueShiftLinearColor(Hue=$User.HueShift, Saturation=1, Value=3, Linear Color=<rgba 1 0.1292 0.05 1>)
     Scale Factor=&MakeVector4FromVectorAndFloat("Vector (RGB)"=<1 1 1>, "Float (Alpha)"=&Multiply_Float(A=1, B=$User.FX_Intensity))
    )
    Particles.DMP_Corona:Vector4f=&MakeVector4(X=1, Y=0, Z=1, W=0)
    Particles.SpriteSize_Corona:Vector2f=&Multiply_Vector2D_ByFloat(
     Float=&Waveform("Waveform [1]"=::NewEnumerator0, "[1] Phase Scale"=0.3, "[1] Frequency"=1, "[1] Amplitude Min/Max"=<0.8 1>)
     Vector2D=<425 425>
    )
   ) {isSetParameters=true}
   /SetVariables(
    Particles.Color_Corona:LinearColor=&MultiplyLinearColor(
     Linear Color=&HueShiftLinearColor(Hue=$User.HueShift, Saturation=1, Value=2, Linear Color=<rgba 1 0.1292 0.05 1>)
     Scale Factor=&MakeVector4FromVectorAndFloat(
      "Vector (RGB)"=<1 1 1>
      "Float (Alpha)"=&FloatFromCurve(
       FloatCurve=DataInterface() {propertyValues {Curve {
          defaultValue=1
          keys [
           {interpMode=RCIM_Cubic, tangentMode=RCTM_User, time=0, value=0}
           {arriveTangent=2.4637, interpMode=RCIM_Cubic, leaveTangent=2.4637, tangentMode=RCTM_User, time=0.5, value=1}
           {tangentMode=RCTM_User, time=1, value=0}
          ]
         }}}
       CurveIndex=&NormalizeFloat(Value=&Divide_Float(A=$Particles.NormalizedAge, B=0.833), Min=0, Max=1)
       Scale Curve=$User.FX_Intensity
      )
     )
    )
    Particles.DMP_Corona:Vector4f=&MakeVector4(
     X=&FloatFromCurve(
      FloatCurve=DataInterface() {propertyValues {Curve {
         defaultValue=1
         keys [
          {arriveTangent=0.4962, interpMode=RCIM_Cubic, leaveTangent=0.4962, tangentMode=RCTM_User, time=0, value=0.5}
          {arriveTangent=1.1785, interpMode=RCIM_Cubic, leaveTangent=1.1785, tangentMode=RCTM_User, time=0.6982, value=0.8936}
          {arriveTangent=0.5646, interpMode=RCIM_Cubic, leaveTangent=0.5646, tangentMode=RCTM_User, time=0.9982, value=1.3744}
         ]
        }}}
      CurveIndex=&NormalizeFloat(Value=&Divide_Float(A=$Particles.NormalizedAge, B=0.833), Min=0, Max=1)
     )
     Y=0
     Z=1
     W=0
    )
    Particles.SpriteSize_Corona:Vector2f=&Multiply_Vector2D_ByFloat(
     Float=&FloatFromCurve(
      FloatCurve=DataInterface() {propertyValues {Curve {
         defaultValue=0
         keys [
          {arriveTangent=4.1785, interpMode=RCIM_Cubic, leaveTangent=4.1785, tangentMode=RCTM_User, time=0, value=0}
          {arriveTangent=1.2018, interpMode=RCIM_Cubic, leaveTangent=1.2018, tangentMode=RCTM_User, time=0.1688, value=0.6228}
          {arriveTangent=3.407, interpMode=RCIM_Cubic, leaveTangent=3.407, tangentMode=RCTM_User, time=0.7314, value=1.6449}
          {arriveTangent=1.177, interpMode=RCIM_Cubic, leaveTangent=1.177, tangentMode=RCTM_User, time=1, value=2.5}
         ]
        }}}
      CurveIndex=&NormalizeFloat(Value=&Divide_Float(A=$Particles.NormalizedAge, B=0.833), Min=0, Max=1)
      Scale Curve=$User.FXScale
     )
     Vector2D=<400 400>
    )
   ) {enabled=false, isSetParameters=true}
   /SetVariables(Particles.Jitter:NiagaraFloat=&Waveform(
     Global Phase=&Add_Float(A=$System.Age, B=&Multiply_FloatByInt(Float=0.375, Integer=&SpawnGroup()))
     "Waveform [1]"=::NewEnumerator12
     "[1] Frequency"=&Lerp_Float(A=5, B=15, Alpha=$System.NormalizedIntensity)
     "[1] Amplitude Min/Max"=<0.9 1>
    )) {isSetParameters=true}
   /SetVariables(Particles.LightPosition:NiagaraPosition=&AddVectorToPosition(Position=$Engine.Emitter.SimulationPosition, Vector=<0 0 60>)) {isSetParameters=true}
   /SetVariables(
    Particles.LightColor=<rgba 8 1.1682 0.1172 1>
    Particles.LightRadius=&Multiply_Float(A=&Multiply_Float(A=$Particles.Jitter, B=1000), B=$User.LightRadiusOverride)
   ) {isSetParameters=true}
   /SetVariables(
    Particles.DynGlow:Vector4f=&MakeVector4(
     X=&Lerp_Float(A=11, B=5, Alpha=$Particles.Jitter)
     Y=&Lerp_Float(
      A=4
      B=30
      Alpha=&FloatFromCurve(
       FloatCurve=DataInterface() {propertyValues {Curve {keys [
           {arriveTangent=0.6328, interpMode=RCIM_Cubic, leaveTangent=0.6328, tangentMode=RCTM_User, time=0, value=0}
           {arriveTangent=2.2092, interpMode=RCIM_Cubic, leaveTangent=2.2092, tangentMode=RCTM_User, time=0.9, value=0.85}
           {interpMode=RCIM_Cubic, leaveTangent=1.512, time=1, value=1}
          ]}}}
       Scale Curve=$System.NormalizedIntensity
      )
     )
     Z=0
     W=0
    )
    Particles.GlowCamOffset:NiagaraFloat=&Lerp_Float(
     A=300
     B=2000
     Alpha=&FloatFromCurve(
      FloatCurve=DataInterface() {propertyValues {Curve {keys [
          {arriveTangent=0.6328, interpMode=RCIM_Cubic, leaveTangent=0.6328, tangentMode=RCTM_User, time=0, value=0}
          {arriveTangent=2.2092, interpMode=RCIM_Cubic, leaveTangent=2.2092, tangentMode=RCTM_User, time=0.9, value=0.85}
          {interpMode=RCIM_Cubic, leaveTangent=1.512, time=1, value=1}
         ]}}}
      Scale Curve=$System.NormalizedIntensity
     )
    )
    Particles.GlowSize=&Multiply_Vector2D_ByFloat(
     Float=&FloatFromCurve(
      FloatCurve=DataInterface() {propertyValues {Curve {keys [
          {arriveTangent=1.7436, interpMode=RCIM_Cubic, leaveTangent=1.7436, tangentMode=RCTM_User, time=0, value=0.5}
          {arriveTangent=0.7173, interpMode=RCIM_Cubic, leaveTangent=0.7173, tangentMode=RCTM_User, time=0.25, value=1}
          {arriveTangent=5.0539, interpMode=RCIM_Cubic, leaveTangent=5.0539, tangentMode=RCTM_User, time=1, value=1}
         ]}}}
      CurveIndex=$System.NormalizedIntensity
      Scale Curve=$Particles.Jitter
     )
     Vector2D=&Vector2DFromFloat(Value=2000)
    )
    Particles.SpriteColor=&MultiplyLinearColor(
     Linear Color=&MultiplyLinearColor(
      Linear Color=<rgba 12 1.7523 0.1758 1>
      Scale Factor=&MakeVector4FromVectorAndFloat(
       "Vector (RGB)"=&VectorFromCurve(
        VectorCurve=DataInterface() {propertyValues {
          XCurve {keys [<at 0 1 RCIM_Cubic>,<at 1 0.3>]}
          YCurve {keys [<at 0.25 1 RCIM_Cubic>,<at 1 0.4 RCIM_Cubic>]}
          ZCurve {keys [<at 0.25 1 RCIM_Cubic>,<at 1 2>]}
         }}
        CurveIndex=$System.NormalizedIntensity
        Scale Curve=&VectorFromFloat(Value=$User.Intensity)
       )
       "Float (Alpha)"=1
      )
     )
     Scale Factor=&MakeVector4(X=1, Y=1, Z=1, W=0.025)
    )
   ) {isSetParameters=true}
  ]
  rendererClasses [
   &NiagaraMeshRendererProperties
   &NiagaraSpriteRendererProperties
   &NiagaraLightRendererProperties
  ]
  renderers [
   {rendererClass=&NiagaraMeshRendererProperties, rendererIndex=0}
   {rendererClass=&NiagaraSpriteRendererProperties, rendererIndex=1}
   {rendererClass=&NiagaraSpriteRendererProperties, rendererIndex=2}
   {rendererClass=&NiagaraLightRendererProperties, rendererIndex=3}
   {rendererClass=&NiagaraSpriteRendererProperties, rendererIndex=4}
  ]
 }
 Mist {rendererClasses [], renderers []}
 RayFlares_ {
  emitterUpdateScript [
   &EmitterState(
    Inactive Response=::NewEnumerator1
    Life Cycle Mode=::NewEnumerator1
    Loop Behavior=::NewEnumerator1
    Loop Duration Mode=::NewEnumerator1
    UseLoopDelay=false
   )
   &SpawnRate(SpawnRate=4, Use Spawn Probability=false)
  ]
  particleSpawnScript [
   /SetVariables(Particles.SpawnRotationVector=&RandomRangeVector(Minimum=<-0.5 0 0>, Maximum=<0.5 0 0>, Evaluation Type=::NewEnumerator1)) {isSetParameters=true}
   &InitializeParticle(
    Lifetime=&RandomRangeFloat(Minimum=1, Maximum=2, Evaluation Type=::NewEnumerator1)
    Mesh Scale=&Multiply_VectorByFloat(
     Vector=/Niagara/DynamicInputs/CascadeConversion/SplitVector(
      Double Component Vector Channels:ECascadeNiagaraTwoVectorChannels=::NewEnumerator1
      Double Component:NiagaraFloat=2.5
      Single Component:NiagaraFloat=1.4
     )
     Float=$User.FXScale
    )
    Mesh Scale Mode=::NewEnumerator4
    Position Offset=<15 -40 35>
    Position Offset Coordinate Space=::Local
    UsePositionOffset=true
    Write Material Random=true
   )
   &InitialMeshOrientation(Use Rotation Vector=false)
  ]
  particleUpdateScript [
   &ParticleState()
   &ScaleMeshSize(Scale Factor=&SplitVector(
     Double Component Vector Channels:ECascadeNiagaraTwoVectorChannels=::NewEnumerator1
     Double Component:NiagaraFloat=&FloatFromCurve(FloatCurve=DataInterface() {propertyValues {Curve {keys [{arriveTangent=-2.0039, interpMode=RCIM_Cubic, time=0, value=2}, <at 1 0.5 RCIM_Cubic>]}}})
     Single Component:NiagaraFloat=&FloatFromCurve(FloatCurve=DataInterface() {propertyValues {Curve {keys [{arriveTangent=0.6532, interpMode=RCIM_Cubic, time=0, value=0.5}, <at 1 1 RCIM_Cubic>]}}})
    ))
   &UpdateMeshOrientation(Rotation Rate=0.25, Rotation Vector=$Particles.SpawnRotationVector)
   &Color(Color=&MakeLinearColorFromVectorAndFloat(
     "Vector (RGB)"=&VectorFromFloat(Value=&FloatFromCurve(
       FloatCurve=DataInterface() {propertyValues {Curve {keys [<at 0 0 RCIM_Cubic>,<at 0.5 1 RCIM_Cubic>,<at 1 0>]}}}
       Scale Curve=&FNTimeOfDayScalar(Night Scale=0.6, Morning Scale=0.9, Evening Scale=0.75)
      ))
     "Float (Alpha)"=&FloatFromCurve(FloatCurve=DataInterface() {propertyValues {Curve {keys [<at 0 0.2 RCIM_Cubic>,<at 1 0 RCIM_Cubic>]}}}, Scale Curve=$User.FX_Intensity)
    ))
   &DynamicMaterialParameters(
    Index 0 Param 1=&FloatFromCurve(FloatCurve=DataInterface() {propertyValues {Curve {keys [<at 0 0>,<at 1 1>]}}})
    Index 0 Param 2=1
    Index 0 Param 3=&FloatFromCurve(
     FloatCurve=DataInterface() {propertyValues {Curve {keys [{arriveTangent=-4.4072, interpMode=RCIM_Cubic, time=0, value=3}, <at 1 1>]}}}
     Scale Curve=1.25
    )
    Index 0 Param 4=1
    Module.Index0Param0Write=true
    Module.Index0Param1Write=true
    Module.Index0Param2Write=true
    Module.Index0Param3Write=true
   )
  ]
  rendererClasses [&NiagaraMeshRendererProperties]
  renderers [{rendererClass=&NiagaraMeshRendererProperties, rendererIndex=0}]
 }
 Rays {rendererClasses [], renderers []}
 Rays001 {rendererClasses [], renderers []}
 Rays002 {rendererClasses [], renderers []}
 SmokeTrail {
  emitterSpawnScript [/SetVariables(Emitter.ParticlePerRibbon=10, Emitter.RibbonNumber=1) {isSetParameters=true}]
  emitterUpdateScript [
   &EmitterState(
    Inactive Response=::NewEnumerator1
    Life Cycle Mode=::NewEnumerator1
    Loop Behavior=::NewEnumerator1
    Loop Duration Mode=::NewEnumerator1
    UseLoopDelay=false
   )
   &SpawnBurst_Instantaneous(
    Spawn Count=&Multiply_Int(A=$Emitter.RibbonNumber, B=$Emitter.ParticlePerRibbon)
    Use Spawn Probability=false
    UseLoopCountLimit=false
   )
  ]
  particleSpawnScript [
   &PartitionParticles(
    ID to Partition:NiagaraInt32=&ReturnExecIndex()
    Number Of Partitions=$Emitter.RibbonNumber
    Particles Per Partition:NiagaraInt32=$Emitter.ParticlePerRibbon
    Partition Mode:ENiagara_PartitionMode=::NewEnumerator1
    Write Partition to Ribbon ID=::NewEnumerator0
   )
   /SetVariables() {isSetParameters=true}
   &InitializeParticle(
    Color Mode=::NewEnumerator0
    Lifetime=5
    Ribbon Width=7
    Ribbon Width Mode=::NewEnumerator1
    UsePositionOffset=false
    Write Material Random=true
   )
   /SetVariables(Particles.RibbonU0Override=$Particles.PartitionParticles.Normalized Index In Partition) {isSetParameters=true}
  ]
  particleUpdateScript [
   &ParticleState(Kill Particles When Lifetime Has Elapsed=false)
   /SetVariables(Particles.RibbonLength=1500) {isSetParameters=true}
   /SetVariables(Particles.RibbonWidth=&Multiply_Float(
     A=$Particles.Initial.RibbonWidth
     B=&Lerp_Float(
      A=15
      B=25
      Alpha=&FloatFromCurve(
       FloatCurve=DataInterface() {propertyValues {Curve {keys [
           {arriveTangent=2.4963, interpMode=RCIM_Cubic, leaveTangent=2.4963, tangentMode=RCTM_User, time=0, value=0.3}
           {arriveTangent=5.8828, interpMode=RCIM_Cubic, leaveTangent=5.8828, tangentMode=RCTM_User, time=0.1, value=0.7}
           {arriveTangent=-0.0177, interpMode=RCIM_Cubic, leaveTangent=-0.0177, tangentMode=RCTM_User, time=0.7, value=2}
          ]}}}
       CurveIndex=$Particles.PartitionParticles.Normalized Index In Partition
       Scale Curve=&Lerp_Float(A=1, B=3, Alpha=0.5)
      )
     )
    )) {isSetParameters=true}
   /SetVariables(Particles.XOffsetValue:NiagaraFloat=&Add_Float(A=$Particles.XOffsetValue, B=&Multiply_Float(A=$Engine.DeltaTime, B=0.075))) {isSetParameters=true}
   /SetVariables(Particles.GoalPos:NiagaraPosition=&AddVectorToPosition(
     Position=&AddVectorToPosition(Position=$Engine.Owner.Position, Vector=<0 0 0>)
     Vector=&Multiply_VectorByFloat(
      Vector=&TransformVector(
       Vector=&Lerp_Vector(A=<-0.2 -0.2 1>, B=<0 0 1>, Alpha=$Particles.PartitionParticles.Normalized Index In Partition)
       Destination Space=::Simulation
      )
      Float=&Lerp_Float(A=0, B=$Particles.RibbonLength, Alpha=$Particles.PartitionParticles.Normalized Index In Partition)
     )
    )) {isSetParameters=true}
   /SetVariables(Particles.Position=&Lerp_Vector(
     A=Linked() {linkedVariable {name=Particles.GoalPos, type=NiagaraPosition}}
     B=&Lerp_Vector(
      A=Linked() {linkedVariable {name=Particles.Position, type=NiagaraPosition}}
      B=Linked() {linkedVariable {name=Particles.GoalPos, type=NiagaraPosition}}
      Alpha=&Multiply_Float(
       A=$Engine.DeltaTime
       B=&Lerp_Float(
        A=15
        B=2
        Alpha=&FloatFromCurve(
         FloatCurve=DataInterface() {propertyValues {Curve {keys [
             {arriveTangent=3.7014, interpMode=RCIM_Cubic, leaveTangent=3.7014, tangentMode=RCTM_User, time=0, value=0}
             {arriveTangent=-0.0363, interpMode=RCIM_Cubic, leaveTangent=-0.0363, tangentMode=RCTM_User, time=1, value=1}
            ]}}}
         CurveIndex=$Particles.PartitionParticles.Normalized Index In Partition
        )
       )
      )
     )
     Alpha=&FloatFromCurve(
      FloatCurve=DataInterface() {propertyValues {Curve {keys [
          {arriveTangent=2.6466, interpMode=RCIM_Cubic, leaveTangent=2.6466, tangentMode=RCTM_User, time=0.1, value=0}
          {arriveTangent=0.2102, interpMode=RCIM_Cubic, leaveTangent=0.2102, tangentMode=RCTM_User, time=1, value=1}
         ]}}}
      CurveIndex=$Particles.PartitionParticles.Normalized Index In Partition
     )
    )) {isSetParameters=true}
   &DynamicMaterialParameters(
    Index 0 Param 1=&Add_Float(
     A=/Game/Effects/Niagara/DynamicInputs/Random/FN_SeededRandomRangeFloat(Minimum:NiagaraFloat=1, Maximum:NiagaraFloat=5, Random Seed:NiagaraInt32=$Particles.PartitionParticles.Partition)
     B=$Particles.XOffsetValue
    )
    Index 0 Param 2=$System.SmoothDiveAlpha
    Index 1 Param 3=&FloatFromCurve(
     FloatCurve=DataInterface() {propertyValues {Curve {keys [<at 0.2 0>,<at 1 2>]}}}
     CurveIndex=&Add_Float(A=$Particles.PartitionParticles.Normalized Index In Partition, B=$System.Timeline.NormalizedTime)
    )
    Module.Index0Param0Write=true
    Module.Index0Param1Write=true
    Module.Index0Param2Write=true
    Module.Index0Param3Write=true
    Module.Index1Param0Write=true
    Module.Index1Param1Write=true
    Module.Index1Param2Write=true
    Module.Index1Param3Write=false
    Write Parameter Index 1=true
   )
   &ScaleColor(
    Scale Alpha=&Multiply_Float(
     A=&FloatFromCurve(
      FloatCurve=DataInterface() {propertyValues {Curve {keys [
          {arriveTangent=-1.6635, interpMode=RCIM_Cubic, leaveTangent=-1.6635, tangentMode=RCTM_User, time=0.15, value=1}
          {arriveTangent=-0.3046, interpMode=RCIM_Cubic, leaveTangent=-0.3046, tangentMode=RCTM_User, time=0.8, value=0}
         ]}}}
      CurveIndex=$Particles.PartitionParticles.Normalized Index In Partition
      Scale Curve=0.5
     )
     B=1
    )
    Scale RGB=&Multiply_VectorByFloat(
     Vector=&MakeVectorFromLinearColorRGB(Color=$System.ToDColor)
     Float=&MakeFloatFromVector(
      Vector=&VectorFromFloat(Value=&FloatFromCurve(
        FloatCurve=DataInterface() {propertyValues {Curve {keys [
            {arriveTangent=-0.0382, interpMode=RCIM_Cubic, leaveTangent=-0.0382, tangentMode=RCTM_User, time=0, value=1}
            {interpMode=RCIM_Cubic, tangentMode=RCTM_User, time=0.5, value=1}
           ]}}}
        CurveIndex=$Particles.PartitionParticles.Normalized Index In Partition
       ))
      Channel=::NewEnumerator0
     )
    )
    ScaleA=true
    ScaleRGB=true
   )
  ]
  rendererClasses [&NiagaraRibbonRendererProperties]
  renderers [{rendererClass=&NiagaraRibbonRendererProperties}]
 }
 Sparkles {
  emitterUpdateScript [
   &EmitterState()
   &SpawnRate(SpawnRate=100, Use Spawn Probability=false)
  ]
  particleSpawnScript [
   &InitializeParticle(
    Color=$User.Color
    Lifetime Max=1.3
    Lifetime Min=0.3
    Lifetime Mode=::NewEnumerator1
    Position Offset=<-20 -20 0>
    Sprite Size Mode=::NewEnumerator1
    Uniform Sprite Size Max=35
    Uniform Sprite Size Min=10
    UsePositionOffset=true
    Write Material Random=true
   )
   &ShapeLocation(
    Cylinder Height=140
    Cylinder Radius=&Lerp_Float(
     A=50
     B=85
     Alpha=&FloatFromCurve(
      FloatCurve=DataInterface() {propertyValues {Curve {keys [
          {interpMode=RCIM_Cubic, tangentMode=RCTM_User, time=0, value=0}
          {arriveTangent=2.2758, interpMode=RCIM_Cubic, leaveTangent=2.2758, tangentMode=RCTM_User, time=1, value=1}
         ]}}}
      CurveIndex=&RandomRangeFloat(Minimum=0, Maximum=1)
     )
    )
    MaskSpawnGroup=false
    Offset=<0 0 80>
    Offset Mode=::NewEnumerator0
    Override Seed=false
    Rotation Coordinate Space=::World
    Shape Primitive=::NewEnumerator1
    Surface Only=false
    UseOwnerScale=true
   )
   &AddVelocity(
    Module.Use Falloff Distance=false
    Origin Offset Coordinate Space=::World
    Override Seed=false
    Velocity Mode=::NewEnumerator1
    Velocity Origin=$Particles.ShapeLocation.ShapePosition
    Velocity Speed=200
   ) {enabled=false}
   &ApplyInitialForces()
  ]
  particleUpdateScript [
   &ParticleState()
   &GravityForce(Gravity=<0 0 -200>)
   /Niagara/Modules/Update/Forces/CurlNoiseForce(Noise Frequency=0.02, Noise Strength=150, Pan Noise=false) {module=CurlNoiseForce001}
   &Drag(Drag=3, Use Linear Drag=true, Use Rotational Drag=false)
   &AccelerationForce() {enabled=false}
   &PointAttractionForce(
    Attraction Strength=&Lerp_Float(
     A=-5
     B=-200
     Alpha=&FloatFromCurve(FloatCurve=DataInterface() {propertyValues {Curve {keys [
          {interpMode=RCIM_Cubic, tangentMode=RCTM_User, time=0, value=0}
          {arriveTangent=2.8893, interpMode=RCIM_Cubic, leaveTangent=2.8893, tangentMode=RCTM_User, time=1, value=1}
         ]}}})
    )
    Attractor Position:NiagaraPosition=&AddVectorToPosition(Position=$Engine.Emitter.SimulationPosition, Vector=<0 0 75>)
    Kill Within Radius=false
    Module.Use Falloff=true
    UseAttractorOffset=false
   )
   &SolveForcesAndVelocity(Clamp Velocity=false, Limit Acceleration=false)
   &ScaleColor(
    Scale Alpha=&FloatFromCurve(FloatCurve=DataInterface() {propertyValues {Curve {keys [<at -0.0092 0.0144>,<at 0.0967 0.9767>,<at 1.0011 0.002>]}}})
    Scale RGB=&VectorFromFloat(Value=&Lerp_Float(
      A=2
      B=10
      Alpha=&FloatFromCurve(
       FloatCurve=DataInterface() {propertyValues {Curve {keys [
           {interpMode=RCIM_Cubic, tangentMode=RCTM_User, time=0, value=0}
           {arriveTangent=1, interpMode=RCIM_Cubic, leaveTangent=1, time=0.8, value=0.25}
           {arriveTangent=6.659, interpMode=RCIM_Cubic, leaveTangent=6.659, tangentMode=RCTM_User, time=1, value=1}
          ]}}}
       CurveIndex=&RandomRangeFloat(Minimum=0, Maximum=1)
      )
     ))
    ScaleA=true
    ScaleRGB=true
   )
   &ScaleSpriteSize(Scale Factor=&Vector2DFromFloat(Value=&FloatFromCurve(FloatCurve=DataInterface() {propertyValues {Curve {keys [<at -0.0046 0>,<at 0.1746 0.9952>,<at 1.0011 0.002>]}}})))
  ]
  rendererClasses [&NiagaraSpriteRendererProperties]
  renderers [{rendererClass=&NiagaraSpriteRendererProperties}]
 }
 Steam {rendererClasses [], renderers []}
}
summary {
 systemName=NS_ToyFox_SupplyDrop_Glow
 emitters [
  {emitterName=RayFlares_, rendererClasses [&NiagaraMeshRendererProperties]}
  {emitterName=SmokeTrail, rendererClasses [&NiagaraRibbonRendererProperties]}
  {
   emitterName=CorePulse
   rendererClasses [
    &NiagaraMeshRendererProperties
    &NiagaraSpriteRendererProperties
    &NiagaraLightRendererProperties
   ]
  }
  {emitterName=Sparkles, rendererClasses [&NiagaraSpriteRendererProperties]}
  {emitterName=Steam, rendererClasses []}
  {emitterName=Mist, rendererClasses []}
  {emitterName=Rays, rendererClasses []}
  {emitterName=Rays001, rendererClasses []}
  {emitterName=Rays002, rendererClasses []}
 ]
 userVariables [
  User.bIsGold {type=NiagaraBool}
  User.Color {type=LinearColor, description="", defaultValue {struct=LinearColor, value=<rgba 1 0.4148 0 1>}}
  User.FX_Intensity {type=NiagaraFloat, defaultValue=1, description=""}
  User.FXScale {type=NiagaraFloat, defaultValue=0.9, description=""}
  User.GlobalOffset {type=Vector3f, description="", defaultValue {struct=Vector3f, value=<0 0 0>}}
  User.HueShift {type=NiagaraFloat, defaultValue=1, description=""}
  User.Intensity {type=NiagaraFloat, defaultValue=0.5, description=""}
  User.LightRadiusOverride {type=NiagaraFloat, defaultValue=0.3, description=""}
  User.Pulse_Speed {type=NiagaraFloat, defaultValue=0.1, description=""}
 ]
}
